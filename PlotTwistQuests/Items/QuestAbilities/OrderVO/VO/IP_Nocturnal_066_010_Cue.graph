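# Sound cue for a Nocturnal voice-over line (IP_Nocturnal_066_010) in the Order VO set of PlotTwist quest ability items.
# A single dialogue player node plays the matching dialogue wave with Nocturnal as speaker and no specified targets.
# It uses the non-critical in-game UI sound class, with boosted volume and a very large max distance.
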SoundCue /PlotTwistQuests/Items/QuestAbilities/OrderVO/VO/IP_Nocturnal_066_010_Cue
root=DialoguePlayer_0
nodes [{
  id=DialoguePlayer_0
  properties {dialogueWaveParameter {
    dialogueWave=&/IP_Nocturnal_066_010_Dialogue
    context {speaker=&/DV_Nocturnal, targets []}
   }}
 }]
properties {maxDistance=2097152, soundClassObject=&Fort_UI_InGame_NonCritical, volumeMultiplier=1.5}
